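# Base material for flattened Athena HLOD proxies, sampling baked PBR textures or constants through static switches, with selectable UV channel and packed-channel options.
# Optional grass replacement masks green-dominant diffuse texels, recolors them from biome or landscape colors, optionally adds snow, and flattens their normals.
# Another switch flattens normals near water height.
# A camera-depth-faded vertical offset feeds world position offset, and MF_HLODWPO processes the final attributes.
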Material /Clyde/Athena/HLODSetup/Materials/BaseFlattenMaterialAthena
graph {
 properties {
  bTangentSpaceNormal=false
  bUseMaterialAttributes=true
  bUsedWithNanite=true
  bUsedWithStaticLighting=true
  pixelDepthOffsetMode=PDOM_Legacy
  usedShadingModels=Default Lit
 }
 nodes [
  n1 = VectorParameter() {parameterName=ReplaceColor, defaultValue=<rgba 0 1 0 0>}
  n2 = Multiply(A=@n1.RGB, B=@n1.RGB)
  n3 = TextureCoordinate() {coordinateIndex=2}
  n4 = TextureCoordinate() {coordinateIndex=5}
  n5 = StaticSwitchParameter(True=TextureCoordinate() {coordinateIndex=6}, False=TextureCoordinate()) {parameterName=UseUV6}
  n6 = StaticSwitchParameter(
   True=TextureCoordinate() {coordinateIndex=3}
   False=StaticSwitchParameter(True=TextureCoordinate() {coordinateIndex=4}, False=StaticSwitchParameter(True=@n4, False=@n5) {parameterName=UseUV5}) {parameterName=UseUV4}
  ) {parameterName=UseUV3}
  n7 = StaticSwitchParameter(
   True=StaticSwitchParameter(True=TextureCoordinate() {coordinateIndex=1}, False=StaticSwitchParameter(True=@n3, False=@n6) {parameterName=UseUV2}) {parameterName=UseUV1}
   False=TextureCoordinate()
  ) {parameterName=UseCustomUV}
  n8 = TextureSampleParameter2D(UVs=@n7) {parameterName=DiffuseTexture, samplerSource=SSM_Clamp_WorldGroupSettings, texture=/Engine/EngineMaterials/BaseFlattenDiffuseMap}
  DotProduct(A=Normalize(VectorInput=@n1.RGB), B=Normalize(VectorInput=Add(A=@n2, B=@n8.RGB)))
  n9 = VectorParameter() {parameterName=ApolloDefaultGrassColor, defaultValue=<rgba 0.6891 0.9323 0.2039 1>}
  n10 = MaterialFunctionCall!(UseLandscapeCoords) {materialFunction=/Game/Athena/Apollo/Environments/Landscape/Materials/Biome/CommonMaterials/MF_ApolloTerrainMasks}
  n11 = Subtract(
   A=ComponentMask(None=WorldPosition().XYZ) {g=true, r=true}
   B=ScalarParameter() {defaultValue=-153000, parameterName=LandscapeMin, sortPriority=-153000}
  )
  n12 = ScalarParameter() {defaultValue=306000, parameterName=LandscapeSize, sortPriority=-153000}
  n13 = TextureSampleParameter2D(UVs=VertexInterpolator(VS=Divide(A=@n11, B=@n12)).PS) {
   parameterName=BiomeColormap
   samplerType=SAMPLERTYPE_LinearColor
   texture=/Game/Athena/Environments/Landscape/Textures/T_Athena_Terrain_CombinedColors_01
  }
  n14 = Multiply(
   A=MaterialFunctionCall!(UseVertexInterp=StaticBool() {value=true}) {materialFunction=/Game/Athena/Environments/Landscape/MaterialFunctions/MF_Landscape_GrassDistanceBlend}.Distant Grass RGB
   B=Multiply(A=@n13.RGB, B=CollectionParameter() {collection=/Game/Athena/Environments/Landscape/MPC/MPC_Landscape, parameterName=ColormapDivisor})
  )
  n15 = Add(A=@n8.G, B=ScalarParameter() {parameterName=GreenBias}) {constB=0.01}
  n16 = Saturate(None=DotProduct(A=ScalarParameter() {parameterName=s}, B=@n8.RGB))
  n17 = ScalarParameter() {defaultValue=2, parameterName=Bmult}
  n18 = Subtract(A=Saturate(None=Divide(A=@n15, B=@n16)), B=Max(A=@n8.R, B=Multiply(A=@n8.B, B=@n17) {constB=0.5}))
  n19 = ScalarParameter() {defaultValue=0.99, parameterName=ReplaceMatchMin}
  n20 = Saturate(None=Divide(A=Subtract(A=@n18, B=@n19), B=ScalarParameter() {defaultValue=0.01, parameterName=ReplaceMatchLen}))
  n21 = LinearInterpolate(
   A=@n8.RGB
   B=StaticSwitchParameter(True=Multiply(A=@n9.RGB, B=@n10.BiomeGrassColors), False=@n14) {defaultValue=true, parameterName=isApolloBiome}
   Alpha=@n20
  )
  n22 = GetMaterialAttributes(None=MaterialFunctionCall!(UseLandscapeCoordsOnSwitch) {materialFunction=/Game/Athena/Environments/Landscape/MaterialFunctions/Snow/MF_Athena_Snow_01}.Material) {attributeGetTypes [#1]}
  n23 = MaterialFunctionCall!(
   Use_ExtraSnowMask=StaticBool() {value=true}
   UV_WorldMask
   SnowStartBiasOverride
   UseHeightLerp
   LandscapeBlend_Height
   Landscape_RockMask
   Snow_Mask=@n20
   Simple_UseWorldZNormal
   Simple_WorldNormal
   WorldPosition
  ) {materialFunction=/Game/Athena/Environments/Landscape/MaterialFunctions/Snow/MF_Athena_Holiday_SnowFall}
  n24 = StaticSwitchParameter(
   True=StaticSwitchParameter(True=LinearInterpolate(A=@n21, B=@n22.Base Color, Alpha=@n23.SnowMask) {constB=0.25}, False=@n21) {parameterName=UseSnow}
   False=@n8.RGB
  ) {parameterName=UseGrassReplacement}
  n25 = VectorParameter() {parameterName=DiffuseConst}
  n26 = TextureSampleParameter2D(UVs=@n7) {parameterName=PackedTexture, samplerSource=SSM_Clamp_WorldGroupSettings, texture=&BaseFlattenDiffuseMap}
  n27 = TextureSampleParameter2D(UVs=@n7) {parameterName=MetallicTexture, samplerSource=SSM_Clamp_WorldGroupSettings, texture=&BaseFlattenDiffuseMap}
  n28 = StaticSwitchParameter(True=@n27.RGB, False=ScalarParameter() {parameterName=MetallicConst}) {parameterName=UseMetallic}
  n29 = TextureSampleParameter2D(UVs=@n7) {parameterName=SpecularTexture, samplerSource=SSM_Clamp_WorldGroupSettings, texture=&BaseFlattenDiffuseMap}
  n30 = StaticSwitchParameter(True=@n29.RGB, False=ScalarParameter() {parameterName=SpecularConst}) {parameterName=UseSpecular}
  n31 = TextureSampleParameter2D(UVs=@n7) {parameterName=RoughnessTexture, samplerSource=SSM_Clamp_WorldGroupSettings, texture=&BaseFlattenDiffuseMap}
  n32 = StaticSwitchParameter(True=@n31.RGB, False=Constant() {r=1}) {parameterName=UseRoughness}
  n33 = Constant3Vector() {constant=<rgba 0 0 1 0>}
  n34 = TextureSampleParameter2D(UVs=@n7) {
   parameterName=NormalTexture
   samplerSource=SSM_Clamp_WorldGroupSettings
   samplerType=SAMPLERTYPE_Normal
   texture=/Engine/EngineMaterials/BaseFlattenNormalMap
  }
  n35 = Transform(None=@n34.RGB)
  n36 = StaticSwitchParameter(True=LinearInterpolate(A=@n35, B=@n33, Alpha=@n20), False=@n35) {parameterName=UseGrassReplacement}
  n37 = WorldPosition()
  n38 = Abs(None=Subtract(A=ComponentMask(None=@n37.XYZ) {b=true}, B=ScalarParameter() {defaultValue=-754, parameterName=WaterZ}) {constB=-754})
  n39 = ScalarParameter() {defaultValue=150, parameterName=FlatRange}
  n40 = Saturate(None=Divide(A=Subtract(A=@n38, B=@n39), B=@n39))
  n41 = LinearInterpolate(A=@n33, B=@n36, Alpha=@n40)
  n42 = MaterialFunctionCall(
   Fade Length=ScalarParameter() {defaultValue=2000, parameterName=Z_Offset_FadeLength}
   Fade Offset=ScalarParameter() {defaultValue=50000, parameterName=Z_Offset_FadeOffset}
   For Vertex Shader=StaticBool() {value=true}
  ) {materialFunction=&CameraDepthFade}
  n43 = AppendVector(A=Constant2Vector(), B=ScalarParameter() {defaultValue=50, parameterName=Z_Offset})
  n44 = TextureSampleParameter2D(UVs=@n7) {
   parameterName=OpacityTexture
   samplerType=SAMPLERTYPE_Grayscale
   texture=/Engine/EngineMaterials/BaseFlattenGrayscaleMap.BaseFlattenGrayScaleMap
  }
  n45 = ScalarParameter() {defaultValue=1, parameterName=OpacityConst}
  n46 = TextureSampleParameter2D(UVs=@n7) {
   parameterName=OpacityMaskTexture
   samplerType=SAMPLERTYPE_Grayscale
   texture=&BaseFlattenGrayScaleMap
  }
  n47 = ScalarParameter() {defaultValue=1, parameterName=OpacityMaskConst}
  n48 = TextureSampleParameter2D(UVs=@n7) {
   parameterName=AmbientOcclusionTexture
   samplerType=SAMPLERTYPE_Grayscale
   texture=&BaseFlattenGrayScaleMap
  }
  n49 = ScalarParameter() {defaultValue=1, parameterName=AmbientOcclusionConst}
  n50 = TextureSampleParameter2D(UVs=@n7) {
   parameterName=EmissiveTexture
   samplerSource=SSM_Clamp_WorldGroupSettings
   samplerType=SAMPLERTYPE_LinearColor
   texture=/Engine/EngineMaterials/BaseFlattenEmissiveMap
  }
  n51 = Multiply(
   A=StaticSwitchParameter(True=@n50.RGB, False=VectorParameter() {parameterName=EmissiveConst}.RGB) {parameterName=UseEmissive}
   B=ScalarParameter() {defaultValue=1, parameterName=EmissiveScale}
  )
  MaterialFunctionCall!(
   IsOcean
   IsTerrain
   In=MakeMaterialAttributes!(
    BaseColor=StaticSwitchParameter(True=@n24, False=@n25.RGB) {parameterName=UseDiffuse}
    Metallic=StaticSwitchParameter(True=@n26.R, False=@n28) {parameterName=PackMetallic}
    Specular=StaticSwitchParameter(True=@n26.B, False=@n30) {parameterName=PackSpecular}
    Roughness=StaticSwitchParameter(True=@n26.G, False=@n32) {parameterName=PackRoughness}
    Anisotropy
    EmissiveColor=@n51
    Opacity=StaticSwitchParameter(True=@n44.RGB, False=@n45) {parameterName=UseOpacity}
    OpacityMask=StaticSwitchParameter(True=@n46.RGB, False=@n47) {parameterName=UseOpacityMask}
    Normal=StaticSwitchParameter(True=@n41, False=@n36) {parameterName=UseWorldZNormalFlattening}
    Tangent
    WorldPositionOffset=Multiply(A=@n42.Result, B=@n43)
    SubsurfaceColor
    ClearCoat
    ClearCoatRoughness
    AmbientOcclusion=StaticSwitchParameter(True=@n48.RGB, False=@n49) {parameterName=UseAmbientOcclusion}
    Refraction
    CustomizedUVs_0
    CustomizedUVs_1
    CustomizedUVs_2
    CustomizedUVs_3
    CustomizedUVs_4
    CustomizedUVs_5
    CustomizedUVs_6
    CustomizedUVs_7
    PixelDepthOffset
    ShadingModel
    Displacement
   )
  ) {materialFunction=/Game/Packages/Fortress_SharedMaterials/Base_Material_Functions/MF_HLODWPO}
  Multiply(A=@n14)
  ScalarParameter() {parameterName=RoughnessConst}
  n52 = Floor(None=OneMinus(None=@n40))
  StaticSwitchParameter(True=Add(A=Multiply(B=@n52), B=@n51), False=@n51) {parameterName=LootLakeEmissive}
  VectorParameter() {parameterName=TESTSte, defaultValue=<rgba 1 0 0 1>}
 ]
}
